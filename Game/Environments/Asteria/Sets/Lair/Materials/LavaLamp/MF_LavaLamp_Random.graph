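# Material function in the Lair set's LavaLamp materials that produces a cheap pseudo-random value from two seeds.
# It multiplies GlobalSeed, left at the default Vector3 input type, by the scalar LocalSeed and outputs the fractional part as Random.
# The result is deterministic for a given seed pair, with no texture lookup or noise node.
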MaterialFunction /Game/Environments/Asteria/Sets/Lair/Materials/LavaLamp/MF_LavaLamp_Random
graph {nodes [
  n1 = FunctionInput() {inputName=GlobalSeed}
  n2 = FunctionInput() {inputName=LocalSeed, inputType=FunctionInput_Scalar}
  FunctionOutput(None=Frac(None=Multiply(A=@n1, B=@n2))) {outputName=Random}
 ]}
