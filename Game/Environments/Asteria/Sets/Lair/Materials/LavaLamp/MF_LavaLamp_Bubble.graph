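# Material function for the lava lamp in the Asteria Lair set that computes the mask for one foreground bubble.
# It compares the tangent-space view direction with a normal derived from the Fragment input's XY, scales the difference by Fragment Z, then divides by a seeded random bubble size.
# A squared falloff outputs a value rising from zero where the directions align, so the bubble shifts with viewing angle.
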MaterialFunction /Game/Environments/Asteria/Sets/Lair/Materials/LavaLamp/MF_LavaLamp_Bubble
graph {nodes [
  n1 = CameraPositionWS()
  n2 = WorldPosition()
  n3 = Transform(None=Normalize(VectorInput=Subtract(A=@n1, B=@n2.XYZ))) {transformSourceType=TRANSFORMSOURCE_World, transformType=TRANSFORM_Tangent}
  n4 = FunctionInput() {inputName=Fragment, sortPriority=1}
  n5 = DeriveNormalZ(InXY=ComponentMask(None=@n4) {g=true, r=true}) {desc="sqrt(saturate(1 - x^2 - y^2))"}
  n6 = Multiply(A=OneMinus(None=DotProduct(A=@n3, B=@n5)), B=ComponentMask(None=@n4) {b=true}) {constB=3000}
  n7 = FunctionInput() {inputName=Seed, inputType=FunctionInput_Scalar, sortPriority=2}
  n8 = LinearInterpolate(
   A=ScalarParameter() {defaultValue=50, group=3. Foreground, parameterName=Foreground Bubble Size Min, sortPriority=6}
   B=ScalarParameter() {defaultValue=1000, group=3. Foreground, parameterName=Foreground Bubble Size Max, sortPriority=7}
   Alpha=MaterialFunctionCall!(
    GlobalSeed=ScalarParameter() {defaultValue=12.654, group=3. Foreground, parameterName=Foreground Seed}
    LocalSeed=Multiply(A=@n7) {constB=1.684}
   ) {materialFunction=&/MF_LavaLamp_Random}.Random
  )
  n9 = OneMinus(None=Saturate(None=Divide(A=@n6, B=@n8) {constB=1000}))
  n10 = Multiply(A=@n9, B=@n9)
  FunctionOutput(None=Multiply(A=OneMinus(None=@n10)) {constB=1.05}) {outputName=Bubble}
 ]}
