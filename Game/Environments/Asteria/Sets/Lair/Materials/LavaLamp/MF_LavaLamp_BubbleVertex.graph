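# Material function that animates one seeded foreground bubble for the lava lamp effect and outputs its view data as "Fragment".
# Seeded random values vary horizontal placement, depth and speed, while a smoothed looping triangle wave drives vertical bobbing and slight sideways drift.
# The tangent-space offset is transformed and added to the object's world position.
# Output packs tangent-space camera direction XY with scaled squared camera distance.
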MaterialFunction /Game/Environments/Asteria/Sets/Lair/Materials/LavaLamp/MF_LavaLamp_BubbleVertex
graph {nodes [
  n1 = CameraPositionWS()
  n2 = MaterialFunctionCall!(
   GlobalSeed=ScalarParameter() {defaultValue=12.654, group=3. Foreground, parameterName=Foreground Seed, sortPriority=99}
   LocalSeed=Multiply(A=NamedRerouteUsage()) {constB=3.698}
  ) {materialFunction=&/MF_LavaLamp_Random}
  n3 = MaterialFunctionCall(
   GlobalSeed=ScalarParameter() {defaultValue=12.654, group=3. Foreground, parameterName=Foreground Seed, sortPriority=99}
   LocalSeed=Multiply(A=NamedRerouteUsage()) {constB=6.984}
  ) {materialFunction=&MF_LavaLamp_Random}
  n4 = Multiply(A=Add(A=@n3.Random) {constB=-0.5}) {constB=0.6}
  n5 = ScalarParameter() {defaultValue=0.01, group=3. Foreground, parameterName=Foreground Speed Min, sliderMax=200, sortPriority=4}
  n6 = ScalarParameter() {defaultValue=0.1, group=3. Foreground, parameterName=Foreground Speed Max, sliderMax=200, sortPriority=5}
  n7 = Frac(None=Multiply(A=Time() {period=1}, B=LinearInterpolate(A=@n5, B=@n6, Alpha=@n3.Random)))
  n8 = Multiply(A=Abs(None=Add(A=@n7) {constB=-0.5})) {constB=2}
  n9 = OneMinus(None=Multiply(A=@n8, B=@n8))
  n10 = Add(A=Multiply(A=@n9, B=@n9)) {constB=-0.5}
  n11 = Multiply(
   A=Add(A=Add(A=@n2.Random) {constB=-0.5}, B=Multiply(A=@n4, B=@n10) {constB=0.2})
   B=ScalarParameter() {group=3. Foreground, parameterName=Foreground Width, sortPriority=0}
  )
  n12 = Multiply(A=@n10, B=ScalarParameter() {defaultValue=100, group=3. Foreground, parameterName=Foreground Height, sortPriority=1})
  n13 = ScalarParameter() {defaultValue=12.654, group=3. Foreground, parameterName=Foreground Seed, sortPriority=99}
  n14 = Multiply(A=NamedRerouteUsage()) {constB=7.654}
  n15 = Multiply(A=LinearInterpolate(
    A=ScalarParameter() {defaultValue=30, group=3. Foreground, parameterName=Foreground Depth Min, sortPriority=2}
    B=ScalarParameter() {defaultValue=150, group=3. Foreground, parameterName=Foreground Depth Max, sortPriority=3}
    Alpha=MaterialFunctionCall(GlobalSeed=@n13, LocalSeed=@n14) {materialFunction=&MF_LavaLamp_Random}.Random
   )) {constB=-1}
  n16 = Add(
   A=Transform(None=MaterialFunctionCall(R=@n11, G=@n12, B=@n15) {materialFunction=&AppendMany}.RGB)
   B=ObjectPositionWS()
  )
  n17 = Transform(None=Normalize(VectorInput=Subtract(A=@n1, B=@n16))) {transformSourceType=TRANSFORMSOURCE_World, transformType=TRANSFORM_Tangent}
  n18 = Distance(A=@n1, B=@n16)
  n19 = Multiply(A=@n18, B=@n18) {constB=2}
  FunctionOutput(None=AppendVector(A=ComponentMask(None=@n17) {g=true, r=true}, B=Multiply(A=@n19) {constB=0.8})) {outputName=Fragment}
  NamedRerouteDeclaration(None=FunctionInput() {inputName=Seed, inputType=FunctionInput_Scalar}) {name=Seed, nodeColor=<rgba 1 0 0.1875 1>}
 ]}
