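# Animated lava lamp panel material for the Asteria Lair set, with glowing bubbles over a vertical color gradient.
# Four scrolling, view-angle-distorted texture samples form the background, while ten seeded bubbles with vertex-interpolated data form the foreground.
# Both layers feed a color function whose warp offsets a tiny-bubble texture.
# An optional ray-box fake-depth glass effect shades the result before emission masking and inverse eye adaptation; Low/Mobile drops the foreground.
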Material /Game/Environments/Asteria/Sets/Lair/Materials/LavaLamp/M_LavaLamp
graph {
 properties {bUsedWithNanite=true, pixelDepthOffsetMode=PDOM_Legacy, usedShadingModels=Default Lit}
 nodes [
  n1 = ScalarParameter() {group=5. Textures, parameterName=Textures Offset X, sortPriority=8}
  n2 = ScalarParameter() {group=5. Textures, parameterName=Textures Offset Y, sortPriority=9}
  n3 = Multiply(
   A=AppendVector(
    A=ScalarParameter() {defaultValue=1, group=5. Textures, parameterName=Textures Tiling X, sortPriority=6}
    B=ScalarParameter() {defaultValue=1, group=5. Textures, parameterName=Textures Tiling Y, sortPriority=7}
   )
   B=Add(A=TextureCoordinate(), B=AppendVector(A=@n1, B=@n2))
  )
  n4 = NamedRerouteDeclaration(None=@n3) {name=Coords, nodeColor=<rgba 0.9688 0 1 1>}
  n5 = TextureSampleParameter2D(UVs=@n4) {
   group=5. Textures
   parameterName=Emission
   samplerType=SAMPLERTYPE_LinearColor
   sortPriority=1
   texture=&/T_LavaLamp_E
  }
  n6 = VectorParameter() {group=1. Colors, parameterName=Panel Color Up, sortPriority=0, defaultValue=<rgba 0.2604 0 0.0048 1>}
  n7 = VectorParameter() {group=1. Colors, parameterName=Panel Color Down, sortPriority=1, defaultValue=<rgba 2 0.1899 0.0104 1>}
  n8 = Power(Base=ComponentMask(None=TextureCoordinate()) {g=true}) {constExponent=3}
  n9 = ScalarParameter() {defaultValue=1.2, group=2. Background, parameterName=Background Tiling X, sortPriority=0}
  n10 = ScalarParameter() {defaultValue=0.6, group=2. Background, parameterName=Background Tiling Y, sortPriority=1}
  n11 = Multiply(A=Add(A=TextureCoordinate()) {constB=-0.5}, B=Multiply(A=AppendVector(A=@n9, B=@n10)) {constB=-1})
  n12 = Normalize(VectorInput=Subtract(A=CameraPositionWS(), B=WorldPosition().XYZ))
  n13 = Multiply(
   A=ComponentMask(None=Transform(None=@n12) {transformSourceType=TRANSFORMSOURCE_World, transformType=TRANSFORM_Tangent}) {g=true, r=true}
   B=ScalarParameter() {defaultValue=1.2, group=2. Background, parameterName=Background Tiling X, sortPriority=0}
  )
  n14 = ScalarParameter() {defaultValue=0.7, group=2. Background, parameterName=Background Angular Distortion, sortPriority=4}
  n15 = LinearInterpolate(
   A=ScalarParameter() {defaultValue=0.4, group=2. Background, parameterName=Background Depth Max, sortPriority=3}
   B=ScalarParameter() {defaultValue=0.1, group=2. Background, parameterName=Background Depth Min, sortPriority=2}
  ) {constAlpha=0}
  n16 = CameraPositionWS()
  n17 = ObjectPositionWS()
  n18 = DotProduct(A=VertexNormalWS(), B=Normalize(VectorInput=Subtract(A=@n16, B=@n17)))
  n19 = Multiply(A=LinearInterpolate(A=@n11, B=@n13, Alpha=LinearInterpolate(A=@n14, B=@n15, Alpha=@n18) {constA=0.9, constB=0.4})) {constB=1.4}
  n20 = ComponentMask(None=@n19) {r=true}
  n21 = Time() {bOverride_Period=true, period=333.3333}
  n22 = Time() {bOverride_Period=true, period=100}
  n23 = AppendVector(
   A=Add(A=Multiply(A=@n20) {constB=-1}, B=Multiply(A=@n21) {constB=0.003})
   B=Add(A=ComponentMask(None=@n19) {g=true}, B=Multiply(A=@n22) {constB=0.01})
  )
  n24 = ScalarParameter() {defaultValue=0.4, group=2. Background, parameterName=Background Depth Max, sortPriority=3}
  n25 = ScalarParameter() {defaultValue=0.1, group=2. Background, parameterName=Background Depth Min, sortPriority=2}
  n26 = LinearInterpolate(A=@n11, B=@n13, Alpha=LinearInterpolate(A=@n14, B=LinearInterpolate(A=@n24, B=@n25) {constAlpha=0.25}, Alpha=@n18) {constA=0.9, constB=0.3})
  n27 = Multiply(A=@n26) {constB=-1.3}
  n28 = Time() {bOverride_Period=true, period=222.2222}
  n29 = ComponentMask(None=@n27) {g=true}
  n30 = Time() {bOverride_Period=true, period=66.6667}
  n31 = AppendVector(
   A=Add(A=ComponentMask(None=@n27) {r=true}, B=Multiply(A=@n28) {constB=-0.0045})
   B=Add(A=Multiply(A=@n29) {constB=-1}, B=Multiply(A=@n30) {constB=-0.015})
  )
  n32 = LinearInterpolate(
   A=ScalarParameter() {defaultValue=0.4, group=2. Background, parameterName=Background Depth Max, sortPriority=3}
   B=ScalarParameter() {defaultValue=0.1, group=2. Background, parameterName=Background Depth Min, sortPriority=2}
  ) {constAlpha=0.75}
  n33 = Multiply(A=LinearInterpolate(A=@n11, B=@n13, Alpha=LinearInterpolate(A=@n14, B=@n32, Alpha=@n18) {constA=0.9, constB=0.2})) {constB=1.2}
  n34 = Time() {bOverride_Period=true, period=166.6667}
  n35 = Time() {bOverride_Period=true, period=50}
  n36 = AppendVector(
   A=Add(A=ComponentMask(None=@n33) {r=true}, B=Multiply(A=@n34) {constB=0.006})
   B=Add(A=ComponentMask(None=@n33) {g=true}, B=Multiply(A=@n35) {constB=0.02})
  )
  n37 = ScalarParameter() {defaultValue=0.4, group=2. Background, parameterName=Background Depth Max, sortPriority=3}
  n38 = ScalarParameter() {defaultValue=0.1, group=2. Background, parameterName=Background Depth Min, sortPriority=2}
  n39 = LinearInterpolate(A=@n11, B=@n13, Alpha=LinearInterpolate(A=@n14, B=LinearInterpolate(A=@n37, B=@n38) {constAlpha=1}, Alpha=@n18) {constA=0.9, constB=0.1})
  n40 = Multiply(A=@n39) {constB=1.1}
  n41 = ComponentMask(None=@n40) {r=true}
  n42 = Multiply(A=Time() {bOverride_Period=true, period=151.5152}) {constB=-0.0066}
  n43 = ComponentMask(None=@n40) {g=true}
  n44 = Multiply(A=Time() {bOverride_Period=true, period=45.4545}) {constB=-0.022}
  n45 = Multiply(A=AppendVector(A=Add(A=@n41, B=@n42), B=Add(A=@n43, B=@n44))) {constB=-1}
  n46 = Multiply(
   A=Multiply(
    A=TextureSample(UVs=@n23) {samplerType=SAMPLERTYPE_LinearGrayscale, texture=&/T_LavaLamp_Background}.R
    B=TextureSample(UVs=@n31) {samplerType=SAMPLERTYPE_LinearGrayscale, texture=&T_LavaLamp_Background}.R
   )
   B=Multiply(
    A=TextureSample(UVs=@n36) {samplerType=SAMPLERTYPE_LinearGrayscale, texture=&T_LavaLamp_Background}.R
    B=TextureSample(UVs=@n45) {samplerType=SAMPLERTYPE_LinearGrayscale, texture=&T_LavaLamp_Background}.R
   )
  )
  n47 = NamedRerouteUsage()
  n48 = Saturate(None=Multiply(A=Multiply(A=@n46, B=@n47)) {constB=2.5})
  n49 = Multiply(A=@n48, B=@n48)
  n50 = Multiply(A=@n49, B=@n49)
  n51 = MaterialFunctionCall!(
   Fresnel=@n50
   Height=Multiply(A=OneMinus(None=@n48), B=ScalarParameter() {defaultValue=15, group=2. Background, parameterName=Background Normal Intensity, sortPriority=6}) {constB=20}
  ) {materialFunction=&/MF_LavaLamp_Colors}
  n52 = ScalarParameter() {defaultValue=0.7, group=2. Background, parameterName=Background Alpha, sortPriority=7}
  n53 = OneMinus(None=@n50)
  n54 = Saturate(None=Multiply(A=@n53) {constB=7})
  n55 = Multiply(A=Multiply(A=@n53, B=@n51.Bubble Warp)) {constB=0.8}
  n56 = Add(A=NamedRerouteUsage(), B=@n55)
  n57 = Multiply(
   A=LinearInterpolate(A=LinearInterpolate(A=@n6.RGB, B=@n7.RGB, Alpha=@n8), B=@n51.Bubble Colors, Alpha=Multiply(A=@n52, B=@n54) {constB=0.6})
   B=Multiply(A=TextureSample(UVs=@n56) {samplerType=SAMPLERTYPE_LinearGrayscale, texture=&/T_LavaLamp_TinyBubbles}.R) {constB=2}
  )
  n58 = Constant() {r=21.57}
  n59 = MaterialFunctionCall!(Seed=@n58) {materialFunction=&/MF_LavaLamp_BubbleVertex}
  n60 = Constant() {r=25.87}
  n61 = MaterialFunctionCall(Seed=@n60) {materialFunction=&MF_LavaLamp_BubbleVertex}
  n62 = Multiply(
   A=MaterialFunctionCall!(Fragment=VertexInterpolator(VS=@n59.Fragment).PS, Seed=@n58) {materialFunction=&/MF_LavaLamp_Bubble}.Bubble
   B=MaterialFunctionCall(Fragment=VertexInterpolator(VS=@n61.Fragment).PS, Seed=@n60) {materialFunction=&MF_LavaLamp_Bubble}.Bubble
  )
  n63 = Constant() {r=9.3647}
  n64 = MaterialFunctionCall(Seed=@n63) {materialFunction=&MF_LavaLamp_BubbleVertex}
  n65 = Constant() {r=14.57}
  n66 = MaterialFunctionCall(Seed=@n65) {materialFunction=&MF_LavaLamp_BubbleVertex}
  n67 = Multiply(
   A=MaterialFunctionCall(Fragment=VertexInterpolator(VS=@n64.Fragment).PS, Seed=@n63) {materialFunction=&MF_LavaLamp_Bubble}.Bubble
   B=MaterialFunctionCall(Fragment=VertexInterpolator(VS=@n66.Fragment).PS, Seed=@n65) {materialFunction=&MF_LavaLamp_Bubble}.Bubble
  )
  n68 = Constant() {r=18.632}
  n69 = MaterialFunctionCall(Seed=@n68) {materialFunction=&MF_LavaLamp_BubbleVertex}
  n70 = Constant() {r=3.2214}
  n71 = MaterialFunctionCall(Seed=@n70) {materialFunction=&MF_LavaLamp_BubbleVertex}
  n72 = Multiply(
   A=MaterialFunctionCall(Fragment=VertexInterpolator(VS=@n69.Fragment).PS, Seed=@n68) {materialFunction=&MF_LavaLamp_Bubble}.Bubble
   B=MaterialFunctionCall(Fragment=VertexInterpolator(VS=@n71.Fragment).PS, Seed=@n70) {materialFunction=&MF_LavaLamp_Bubble}.Bubble
  )
  n73 = Constant() {r=9.852}
  n74 = MaterialFunctionCall(Seed=@n73) {materialFunction=&MF_LavaLamp_BubbleVertex}
  n75 = Constant() {r=11.566}
  n76 = MaterialFunctionCall(Seed=@n75) {materialFunction=&MF_LavaLamp_BubbleVertex}
  n77 = Multiply(
   A=MaterialFunctionCall(Fragment=VertexInterpolator(VS=@n74.Fragment).PS, Seed=@n73) {materialFunction=&MF_LavaLamp_Bubble}.Bubble
   B=MaterialFunctionCall(Fragment=VertexInterpolator(VS=@n76.Fragment).PS, Seed=@n75) {materialFunction=&MF_LavaLamp_Bubble}.Bubble
  )
  n78 = Constant() {r=4.876}
  n79 = VertexInterpolator(VS=MaterialFunctionCall(Seed=@n78) {materialFunction=&MF_LavaLamp_BubbleVertex}.Fragment)
  n80 = Constant() {r=8.3789}
  n81 = VertexInterpolator(VS=MaterialFunctionCall(Seed=@n80) {materialFunction=&MF_LavaLamp_BubbleVertex}.Fragment)
  n82 = Multiply(
   A=Multiply(A=Multiply(A=@n62, B=@n67), B=Multiply(A=@n72, B=@n77))
   B=Multiply(
    A=MaterialFunctionCall(Fragment=@n79.PS, Seed=@n78) {materialFunction=&MF_LavaLamp_Bubble}.Bubble
    B=MaterialFunctionCall(Fragment=@n81.PS, Seed=@n80) {materialFunction=&MF_LavaLamp_Bubble}.Bubble
   )
  )
  n83 = NamedRerouteUsage()
  n84 = Saturate(None=Multiply(A=Multiply(A=@n82, B=@n83)) {constB=1.5})
  n85 = Multiply(A=@n84, B=@n84)
  n86 = Multiply(A=@n85, B=@n85)
  n87 = Multiply(A=@n86, B=@n86)
  n88 = MaterialFunctionCall(
   Fresnel=@n87
   Height=Multiply(A=OneMinus(None=@n84), B=ScalarParameter() {defaultValue=18, group=3. Foreground, parameterName=Foreground Normal Intensity, sortPriority=8}) {constB=15}
  ) {materialFunction=&MF_LavaLamp_Colors}
  n89 = OneMinus(None=@n87)
  n90 = Multiply(A=OneMinus(None=Multiply(A=@n87) {constB=0.8}), B=Saturate(None=Multiply(A=@n89) {constB=10}))
  n91 = Multiply(A=Multiply(A=@n89, B=@n88.Bubble Warp)) {constB=0.8}
  n92 = TextureSample(UVs=Add(A=LinearInterpolate(A=@n55, B=@n91, Alpha=@n90), B=NamedRerouteUsage())) {samplerType=SAMPLERTYPE_LinearGrayscale, texture=&T_LavaLamp_TinyBubbles}
  n93 = MaterialFunctionCall!(Default=Multiply(A=LinearInterpolate(A=@n57, B=@n88.Bubble Colors, Alpha=@n90), B=Multiply(A=@n92.R) {constB=2}), Low/Mobile=@n57) {materialFunction=/Game/Effects/Fort_Effects/Materials/_Functions/MF_LowMobile_PathingSwitch}
  n94 = AppendVector(
   A=Multiply(A=ScalarParameter() {defaultValue=1, group=5. Textures, parameterName=Textures Tiling X, sortPriority=6}) {constB=2}
   B=ScalarParameter() {defaultValue=1, group=5. Textures, parameterName=Textures Tiling Y, sortPriority=7}
  )
  n95 = Divide(
   A=AppendVector(
    A=ScalarParameter() {defaultValue=186, group=4. Glass, parameterName=Glass Panel Size X, sortPriority=1}
    B=ScalarParameter() {defaultValue=360, group=4. Glass, parameterName=Glass Panel Size Y, sortPriority=2}
   )
   B=@n94
  )
  n96 = ScalarParameter() {defaultValue=20, group=4. Glass, parameterName=Glass Depth, sortPriority=3}
  n97 = Multiply(
   A=Divide(B=AppendVector(A=@n95, B=@n96))
   B=Transform(None=CameraVectorWS()) {desc=Tangent, transformSourceType=TRANSFORMSOURCE_World, transformType=TRANSFORM_Tangent}
  )
  n98 = Divide(B=@n97) {desc=id}
  n99 = NamedRerouteUsage()
  n100 = Frac(None=Multiply(A=@n99, B=Constant2Vector() {g=-1, r=-2}))
  n101 = AppendVector(A=Add(A=Multiply(A=@n100) {constB=2}) {constB=-1}, B=Constant() {r=-1})
  n102 = MaterialFunctionCall(Float3=Subtract(A=Abs(None=@n98), B=Multiply(A=@n98, B=@n101)) {desc=k}) {materialFunction=&BreakOutFloat3Components}
  n103 = Min(A=Min(A=@n102.R, B=@n102.R), B=@n102.B)
  n104 = ComponentMask(None=Add(A=Multiply(A=@n103, B=@n97), B=@n101) {desc="pos.xyz += kMin * In.dir;"}) {b=true}
  n105 = Saturate(None=Multiply(A=Multiply(A=@n104) {constB=-1}, B=ScalarParameter() {defaultValue=5, group=4. Glass, parameterName=Glass Sharpness, sortPriority=4}) {constB=2})
  n106 = Multiply(A=@n105, B=@n105) {constB=0.3}
  n107 = Add(A=Multiply(A=ScalarParameter() {defaultValue=0.3, group=4. Glass, parameterName=Glass Brighness, sortPriority=6}, B=@n106) {constB=0.3})
  n108 = Multiply(
   A=StaticSwitchParameter(True=Multiply(A=@n93.Result, B=@n107), False=@n93.Result) {defaultValue=true, group=4. Glass, parameterName=Glass Fake Depth, sortPriority=0}
   B=ScalarParameter() {defaultValue=1, group=1. Colors, parameterName=Color Multiplier}
  )
  EyeAdaptationInverse(
   LightValueInput=Multiply(A=@n5.R, B=@n108)
   AlphaInput=ScalarParameter() {defaultValue=0.7, group=1. Colors, parameterName=Inverse Eye Adaptation, sortPriority=23}
  )
  Multiply(A=@n99, B=Multiply(A=@n94) {constB=-1}) {constB=-5}
  n109 = VectorParameter() {group=1. Colors, parameterName=Base Color, sortPriority=20, defaultValue=<rgba 0.0561 0.063 0.0529 1>}
  n110 = Desaturation(None=OneMinus(None=@n108))
  Multiply(A=@n109.RGB, B=Multiply(A=@n110, B=@n110))
  Multiply(
   A=Time() {bOverride_Period=true, period=1}
   B=ScalarParameter() {defaultValue=1, group=2. Background, parameterName=Background Speed, sortPriority=5}
  )
  n111 = Multiply(A=@n39) {constB=4}
  n112 = Multiply(A=ComponentMask(None=@n111) {g=true}) {constB=-1}
  n113 = Multiply(A=Time()) {constB=0.32}
  NamedRerouteDeclaration(None=AppendVector(A=ComponentMask(None=@n111) {r=true}, B=Add(A=@n112, B=@n113))) {name=TinyBubblesCoord2, nodeColor=<rgba 0 0.7891 1 1>}
  n114 = Multiply(A=@n26) {constB=10}
  n115 = Multiply(A=ComponentMask(None=@n114) {g=true}) {constB=-1}
  n116 = Multiply(A=Time()) {constB=0.2}
  NamedRerouteDeclaration(None=AppendVector(A=ComponentMask(None=@n114) {r=true}, B=Add(A=@n115, B=@n116))) {name=TinyBubblesCoord1, nodeColor=<rgba 1 0.4688 0 1>}
  n117 = TextureSampleParameter2D(UVs=@n4) {
   group=5. Textures
   parameterName=Normal
   samplerType=SAMPLERTYPE_Normal
   sortPriority=3
   texture=&/T_LavaLamp_N
  }
  n118 = Multiply(
   A=@n5.G
   B=Multiply(
    A=ComponentMask(None=@n117.RGB) {r=true}
    B=ScalarParameter() {defaultValue=1, group=4. Glass, parameterName=Glass Refraction, sortPriority=-1}
   )
  )
  n119 = Multiply(A=@n106, B=ScalarParameter() {defaultValue=0.3, group=4. Glass, parameterName=Glass Distortion, sortPriority=5}) {constB=0.3}
  n120 = Add(
   A=@n118
   B=MaterialFunctionCall(Default=OneMinus(None=@n119), Low/Mobile=Constant() {r=1}) {materialFunction=&MF_LowMobile_PathingSwitch}.Result
  )
  n121 = Add(A=@n118)
  NamedRerouteDeclaration(None=StaticSwitchParameter(
    True=MaterialFunctionCall(Default=@n120, Low/Mobile=@n121) {materialFunction=&MF_LowMobile_PathingSwitch}.Result
    False=@n121
   ) {defaultValue=true, group=4. Glass, parameterName=Glass Fake Depth, sortPriority=0}) {name=Normal Distort, nodeColor=<rgba 0 1 0.9766 1>}
  Reroute(None=@n3)
  ScalarParameter() {defaultValue=0.6, group=1. Colors, parameterName=Specular, sortPriority=21}
  TextureCoordinate()
  TextureSampleParameter2D(UVs=@n4) {
   group=5. Textures
   parameterName=Roughness
   samplerType=SAMPLERTYPE_LinearColor
   sortPriority=0
   texture=&/T_LavaLamp_R
  }
 ]
}
